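# Translucent, unlit, two-sided toon outline material for the Weeping Woods character, drawing screen-space lines as opacity with switchable outline colors.
# Four neighboring SceneDepth samples, offset by distance- or world-size-based thickness, feed combine4Samples for edge detection.
# Pixel-versus-scene depth and optional subsurface GBuffer masking limit lines; roughness edges can add more.
# Shading-path switches provide mobile fallbacks, normal-based WPO optionally inflates the mesh, and an emissive sweep branch stays unconnected to output.
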Material /Game/Characters/Player/Male/Medium/Bodies/M_MED_Weeping_Woods_Toon/Materials/M_BlacknWhite_Outline_V2
graph {
 properties {
  bEnableResponsiveAA=true
  bUseMaterialAttributes=true
  bUsedWithMeshDeformer=true
  bUsedWithMorphTargets=true
  bUsedWithNiagaraMeshParticles=true
  bUsedWithSkeletalMesh=true
  blendMode=BLEND_Translucent
  pixelDepthOffsetMode=PDOM_Legacy
  shadingModel=MSM_Unlit
  twoSided=true
  usedShadingModels=Unlit
 }
 nodes [
  Constant() {r=0.001}
  Constant() {r=2}
  Constant() {r=0.75}
  n1 = NamedRerouteUsage()
  LinearInterpolate(Alpha=@n1)
  n2 = NamedRerouteUsage()
  n3 = Desaturation(None=SceneTexture() {sceneTextureId=PPI_Metallic}.Color)
  n4 = Constant() {r=0.5}
  n5 = Constant() {r=1}
  n6 = Constant()
  n7 = LinearInterpolate(
   B=VectorParameter() {parameterName=SecondaryOutline, defaultValue=<rgba 1 1 1 0>}.RGB
   Alpha=Multiply(A=Saturate(None=@n2), B=If(A=@n3, B=@n4, "A > B"=@n5, "A == B"=@n6, "A < B"=@n6))
  )
  n8 = VectorParameter() {parameterName=OutlineColor, defaultValue=<rgba 0 0 0 1>}
  n9 = VectorParameter() {parameterName=MobileOutlineColor}
  n10 = ShadingPathSwitch(
   Default=StaticSwitchParameter(True=StaticSwitchParameter(True=@n7, False=@n8.RGB) {parameterName=UseSecondaryOutlineColor}, False=Constant()) {parameterName=UseOutlineColor}
   Forward=@n9.RGB
   Mobile=@n9.RGB
  )
  n11 = SceneDepth() {desc=SceneDepth works on mobile whereas SceneTexture doesn't}
  n12 = Subtract(A=PixelDepth().R, B=@n11.R)
  n13 = Multiply(
   A=ScalarParameter() {defaultValue=3, group=Cartoon, parameterName=CRTNO_Outline-InnerMaskBias}
   B=Divide(A=ObjectRadius(), B=ScalarParameter() {defaultValue=97, parameterName=ObjectScale})
  )
  n14 = NamedRerouteUsage()
  n15 = LinearInterpolate(
   A=ScalarParameter() {defaultValue=16, parameterName=SampleBias_Min}
   B=ScalarParameter() {defaultValue=16, parameterName=SampleBias}
   Alpha=@n14
  )
  n16 = LinearInterpolate(
   A=ScalarParameter() {defaultValue=1, parameterName=MobileSampleBias_Min}
   B=ScalarParameter() {defaultValue=16, parameterName=Mobile_SampleBias}
   Alpha=@n14
  )
  n17 = ComponentMask(None=ScreenPosition().ViewportUV) {g=true, r=true}
  n18 = MaterialFunctionCall!(
   WorldUnits=LinearInterpolate(
    A=ScalarParameter() {defaultValue=0.07, group=Cartoon, parameterName=CRTNO_Outline-InnerSize1_WorldUnits}
    B=ScalarParameter() {defaultValue=1, group=Cartoon, parameterName=CRTNO_Outline-InnerSize2_WorldUnits}
    Alpha=NamedRerouteUsage()
   )
   Depth
  ) {materialFunction=/Engine/Functions/Engine_MaterialFunctions01/Coordinates/WorldUnitsInPixel}
  n19 = ScalarParameter() {defaultValue=0.51, parameterName=CRTNO_Outline-InnerSize1_MinPixels}
  n20 = ScalarParameter() {defaultValue=4, parameterName=CRTNO_Outline-InnerSize1_MaxPixels}
  n21 = Divide(A=ViewSize(), B=Constant() {r=320})
  n22 = Constant() {r=2}
  n23 = Clamp(None=Ceil(None=Subtract(A=@n21, B=@n22) {constB=0})) {maxDefault=2, minDefault=1}
  n24 = Divide(
   A=Subtract(A=PixelDepth().R, B=ScalarParameter() {defaultValue=50, group=Cartoon, parameterName=CRTNO_Outline-InnerScaleStart})
   B=ScalarParameter() {defaultValue=500, group=Cartoon, parameterName=CRTNO_Outline-InnerScaleDistance}
  )
  n25 = LinearInterpolate(
   A=ScalarParameter() {defaultValue=2, group=Cartoon, parameterName=CRTNO_Outline-InnerSize1}
   B=ScalarParameter() {defaultValue=0.75, group=Cartoon, parameterName=CRTNO_Outline-InnerSize2}
   Alpha=NamedRerouteDeclaration(None=Clamp(None=@n24)) {name=DepthMask, nodeColor=<rgba 0.6406 0 1 1>}
  )
  n26 = Multiply(A=@n25, B=ScalarParameter() {defaultValue=1.5, group=Cartoon, parameterName=CRTNO_Outline-InnerSizeMobileMultiplier})
  n27 = Divide(A=Constant() {r=960}, B=ComponentMask(None=ViewSize()) {r=true})
  n28 = Max(A=ShadingPathSwitch(Default=@n25, Forward=@n26, Mobile=@n26), B=Saturate(None=Floor(None=@n27)) {desc=if resolution is smaller than 960H}) {desc="If res smaller than target, calmp min value to 1"}
  n29 = Multiply(
   A=StaticSwitchParameter(True=Clamp(None=@n18.PIxels, Min=@n19, Max=@n20), False=Multiply(A=@n23, B=@n28)) {group=Cartoon, parameterName=OutlineThickness_InWorldUnits}
   B=SceneTexelSize()
  ) {constA=2}
  n30 = Constant()
  n31 = AppendVector(A=ComponentMask(None=@n29) {desc=x, r=true}, B=@n30)
  n32 = Add(A=@n17, B=@n31) {desc=x+}
  n33 = SceneDepth(UVs=@n32)
  n34 = AppendVector(A=@n30, B=ComponentMask(None=@n29) {desc=y, g=true}) {desc=y+}
  n35 = Add(A=@n17, B=@n34) {desc=y+}
  n36 = SceneDepth(UVs=@n35)
  n37 = Constant() {r=-1}
  n38 = Add(A=@n17, B=Multiply(A=@n31, B=@n37)) {desc=x-}
  n39 = SceneDepth(UVs=@n38)
  n40 = Add(A=@n17, B=Multiply(A=@n37, B=@n34)) {desc=y-}
  n41 = SceneDepth(UVs=@n40)
  n42 = NamedRerouteUsage()
  n43 = LinearInterpolate(
   A=ScalarParameter() {defaultValue=-1, parameterName=Divisor_Min}
   B=ScalarParameter() {defaultValue=-16, parameterName=Divisor_max}
   Alpha=@n42
  )
  n44 = LinearInterpolate(
   A=ScalarParameter() {defaultValue=-1, parameterName=MobileDivisor_Min}
   B=ScalarParameter() {defaultValue=-16, parameterName=MobileDivisor_max}
   Alpha=@n42
  )
  n45 = ComponentMask(None=MaterialFunctionCall!(
    Base=@n11.R
    BaseSampleBias=ShadingPathSwitch(Default=@n15, Mobile=@n16)
    Scale=ScalarParameter() {defaultValue=4, parameterName=Scale}
    Sample1=@n33.R
    Sample2=@n36.R
    Sample3=@n39.R
    Sample4=@n41.R
    Divisor=ShadingPathSwitch(Default=@n43, Mobile=@n44)
    UseCenteredLines
    UseRGB
   ) {materialFunction=/Game/Packages/jl_tests/combine4Samples}.Result) {r=true}
  n46 = SceneTexture() {sceneTextureId=PPI_Roughness}
  n47 = LinearInterpolate(A=ScalarParameter() {parameterName=SampleBias_Min_RGH}, B=ScalarParameter() {parameterName=SampleBias_RGH}, Alpha=NamedRerouteUsage())
  n48 = Constant() {r=4}
  n49 = SceneTexture(UVs=NamedRerouteUsage()) {sceneTextureId=PPI_Roughness}
  n50 = SceneTexture(UVs=NamedRerouteUsage()) {sceneTextureId=PPI_Roughness}
  n51 = SceneTexture(UVs=NamedRerouteUsage()) {sceneTextureId=PPI_Roughness}
  n52 = SceneTexture(UVs=NamedRerouteUsage()) {sceneTextureId=PPI_Roughness}
  n53 = ScalarParameter() {defaultValue=-0.04, group=RoughnessGBufferLines, parameterName=InnerLinesDivisor}
  n54 = Max(
   A=@n45
   B=ComponentMask(None=MaterialFunctionCall(
     Base=@n46.Color
     BaseSampleBias=@n47
     Scale=@n48
     Sample1=@n49.Color
     Sample2=@n50.Color
     Sample3=@n51.Color
     Sample4=@n52.Color
     Divisor=@n53
    ) {materialFunction=&combine4Samples}.Result) {r=true}
  )
  n55 = Multiply(
   A=Saturate(None=Add(A=@n12, B=@n13))
   B=StaticSwitchParameter(True=ShadingPathSwitch(Default=@n54, Forward=@n45, Mobile=@n45), False=@n45) {group=RoughnessGBufferLines, parameterName=UseRoughnessGBufferOutlines}
  )
  n56 = Multiply(A=NamedRerouteUsage(), B=@n55)
  n57 = Multiply(
   A=Saturate(None=StaticSwitchParameter(True=@n56, False=@n55) {defaultValue=true, group=RoughnessGBufferLines, parameterName=UseSubsurfaceMasking})
   B=ScalarParameter() {defaultValue=1, group=Cartoon, parameterName=CRTNO_Outline-InnerAmount}
  )
  n58 = Constant()
  n59 = Saturate(None=StaticSwitchParameter(True=@n57, False=ShadingPathSwitch(Default=@n57, Forward=@n58, Mobile=@n58)) {group=Cartoon, parameterName=UseDepthBufferOutlinesOnMobile})
  n60 = StaticSwitchParameter(True=Multiply(A=VertexNormalWS(), B=ScalarParameter() {defaultValue=0.5, parameterName="ScaleMesh\\"}), False=Constant()) {parameterName=Use WPO}
  MaterialFunctionCall(In=SetMaterialAttributes!(
    MaterialAttributes
    Emissive Color=NamedRerouteDeclaration(None=@n10) {name=emissive, nodeColor=<rgba 1 0.8672 0 1>}
    Opacity=NamedRerouteDeclaration(None=@n59) {name=opacity, nodeColor=<rgba 0 1 0.5078 1>}
    World Position Offset=NamedRerouteDeclaration(None=@n60) {name=wpo, nodeColor=<rgba 0 1 0.5547 1>}
   ) {attributeSetTypes [#1, #2, #3]}) {materialFunction=&MF_TechArtOverrides}
  MaterialFunctionCall() {materialFunction=&ComponentPivotLocation}
  n61 = LinearInterpolate(
   A=ScalarParameter() {defaultValue=16, parameterName=SampleBias_Min_Outer}
   B=ScalarParameter() {defaultValue=16, parameterName=SampleBias_Outer}
   Alpha=@n1
  )
  n62 = ComponentMask(None=MaterialFunctionCall(
    Base=@n11.R
    BaseSampleBias=@n61
    Scale=ScalarParameter() {defaultValue=1, parameterName=Scale_depth2}
    Sample1=@n33.R
    Sample2=@n36.R
    Sample3=@n39.R
    Sample4=@n41.R
    Divisor=ScalarParameter() {defaultValue=-1, parameterName=Divisor_depth2}
   ) {materialFunction=&combine4Samples}.Result) {r=true}
  NamedRerouteDeclaration(None=OneMinus(None=Saturate(None=@n62))) {name=Inner Outline Mask, nodeColor=<rgba 1 0 0.9982 1>}
  n63 = SceneTexture() {sceneTextureId=PPI_SubsurfaceColor}
  n64 = Constant()
  n65 = If(
   A=Min(A=ComponentMask(None=@n63.Color) {g=true}, B=ComponentMask(None=@n63.Color) {b=true})
   B=SquareRoot(None=Constant() {r=0.001})
   "A > B"=@n64
   "A == B"=Constant() {r=1}
   "A < B"=@n64
  ) {equalsThreshold=0.0005}
  n66 = Constant() {desc=for PS4 and XO, r=1}
  n67 = QualitySwitch(Default=@n65, Low=Floor(None=SceneTexture() {sceneTextureId=PPI_Opacity}.Color), High=@n66, Medium=@n66, Epic=@n65)
  n68 = Constant() {r=1}
  NamedRerouteDeclaration(None=ShadingPathSwitch(Default=@n67, Deferred=@n67, Forward=@n68, Mobile=@n68)) {name=Subsurface Mask, nodeColor=<rgba 0.5469 0 1 1>}
  n69 = Multiply(
   A=Subtract(A=WorldPosition().XYZ, B=ObjectPositionWS())
   B=ScalarParameter() {defaultValue=0.001, group=Emissive Sweep, parameterName=WorldTileScale}
  )
  n70 = Normalize(VectorInput=Transform(None=VectorParameter() {group=Emissive Sweep, parameterName=SweepDirectionCamSpace, defaultValue=<rgba -0.5 0.45 -1 0>}.RGB) {transformSourceType=TRANSFORMSOURCE_Camera})
  n71 = ComponentMask(None=Multiply(A=DotProduct(A=@n69, B=@n70), B=@n70)) {b=true, r=true}
  n72 = Constant() {r=0.28}
  n73 = Frac(None=ComponentMask(None=Panner(Coordinate=@n71, Speed=@n72) {speedX=0.2, speedY=0.2}) {g=true})
  n74 = Add(A=@n73) {constB=0.1}
  n75 = OneMinus(None=Multiply(A=Subtract(A=@n73) {constB=0.9}) {constB=10})
  n76 = MaterialFunctionCall() {materialFunction=&MF_Fresnel_FX}
  n77 = Multiply(A=Multiply(A=Floor(None=@n74), B=Power(Base=@n75) {constExponent=3}), B=Saturate(None=Add(A=@n76.Fresnel Inverted) {constB=0.1}))
  n78 = VectorParameter() {group=Emissive Sweep, parameterName=SweepColour, defaultValue=<rgba 1.8958 2 1.9078 0>}
  n79 = LinearInterpolate(
   A=@n78.RGB
   B=Multiply(A=@n78.RGB, B=ScalarParameter() {defaultValue=15, group=Emissive Sweep, parameterName=FBoostScale})
   Alpha=VertexColor().G
  )
  NamedRerouteDeclaration(None=StaticSwitchParameter(True=Multiply(A=@n77, B=@n79), False=Constant()) {group=Emissive Sweep, parameterName=UseSweep}) {name=EmissiveSweep, nodeColor=<rgba 1 0.5859 0 1>}
  NamedRerouteDeclaration(None=@n32) {name=Sample1, nodeColor=<rgba 1 0.2344 0 1>}
  NamedRerouteDeclaration(None=@n35) {name=Sample2, nodeColor=<rgba 1 0.2344 0 1>}
  NamedRerouteDeclaration(None=@n38) {name=Sample3, nodeColor=<rgba 1 0.2344 0 1>}
  NamedRerouteDeclaration(None=@n40) {name=Sample4, nodeColor=<rgba 1 0.2344 0 1>}
  ShadingPathSwitch(
   Default=@n61
   Mobile=LinearInterpolate(
    A=ScalarParameter() {defaultValue=1, parameterName=MobileSampleBias_Min}
    B=ScalarParameter() {defaultValue=16, parameterName=Mobile_SampleBias}
    Alpha=@n14
   )
  )
 ]
}
